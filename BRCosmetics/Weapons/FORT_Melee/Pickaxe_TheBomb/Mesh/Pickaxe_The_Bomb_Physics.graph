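# Physics asset for The Bomb two-handed pickaxe, defining capsule bodies for its handle and two dangling chains.
# Chain A has three bodies and chain B has six, each starting at a tiny root capsule followed by uniform link capsules.
# Constraints join neighboring links within each chain, with free swing at the first joint and tighter swing limits further down.
# No constraint connects the chain roots to the handle or to each other.
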PhysicsAsset /BRCosmetics/Weapons/FORT_Melee/Pickaxe_TheBomb/Mesh/Pickaxe_The_Bomb_Physics
bodies {
 Melee_TwoHanded_Handle [{shapeType=Capsule, length=87.0153, radius=2.8624, center=<0 -1.3853 0>}]
 dyn_chain_a01 [{shapeType=Capsule, length=0, radius=0.5, rotation=<rot 0 -90 0>}]
 dyn_chain_a02 [{shapeType=Capsule, length=3.7, radius=2.6, center=<2.8 0 0>, rotation=<rot 90 0 0>}]
 dyn_chain_a03 [{shapeType=Capsule, length=3.7, radius=2.6, center=<2.8 0 0>, rotation=<rot 90 0 0>}]
 dyn_chain_b01 [{shapeType=Capsule, length=0, radius=0.5, center=<1.9982 0 0>, rotation=<rot 90 0.0264 0.0264>}]
 dyn_chain_b02 [{shapeType=Capsule, length=3.7, radius=2.6, center=<2.8 0 0>, rotation=<rot 90 0 0>}]
 dyn_chain_b03 [{shapeType=Capsule, length=3.7, radius=2.6, center=<2.8 0 0>, rotation=<rot 90 0 0>}]
 dyn_chain_b04 [{shapeType=Capsule, length=3.7, radius=2.6, center=<2.8 0 0>, rotation=<rot 90 0 0>}]
 dyn_chain_b05 [{shapeType=Capsule, length=3.7, radius=2.6, center=<2.8 0 0>, rotation=<rot 90 0 0>}]
 dyn_chain_b06 [{shapeType=Capsule, length=3.7, radius=2.6, center=<2.8 0 0>, rotation=<rot 90 0 0>}]
}
constraints [
 {
  bone1Name=dyn_chain_a02
  bone2Name=dyn_chain_a01
  swing1LimitDegrees=135
  swing1Motion=Free
  swing2LimitDegrees=135
  swing2Motion=Free
  twistLimitDegrees=20
 }
 {bone1Name=dyn_chain_a03, bone2Name=dyn_chain_a02, swing1LimitDegrees=60, swing2LimitDegrees=60, twistLimitDegrees=20}
 {
  bone1Name=dyn_chain_b02
  bone2Name=dyn_chain_b01
  swing1LimitDegrees=90
  swing1Motion=Free
  swing2LimitDegrees=90
  swing2Motion=Free
  twistLimitDegrees=20
 }
 {bone1Name=dyn_chain_b03, bone2Name=dyn_chain_b02, swing1LimitDegrees=60, swing2LimitDegrees=60, twistLimitDegrees=20}
 {bone1Name=dyn_chain_b04, bone2Name=dyn_chain_b03, swing1LimitDegrees=60, swing2LimitDegrees=60, twistLimitDegrees=20}
 {bone1Name=dyn_chain_b05, bone2Name=dyn_chain_b04, swing1LimitDegrees=60, swing2LimitDegrees=60, twistLimitDegrees=20}
 {bone1Name=dyn_chain_b06, bone2Name=dyn_chain_b05, swing1LimitDegrees=60, swing2LimitDegrees=60, twistLimitDegrees=20}
]
